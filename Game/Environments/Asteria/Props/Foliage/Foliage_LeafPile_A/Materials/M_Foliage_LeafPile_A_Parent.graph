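# Masked, two-sided, Nanite-enabled parent material for the Asteria leaf pile foliage prop.
# The U coordinate of UV1 selects between two parameterized texture sets, tested against 1, for base color, opacity, normal, specular and height, so one material covers both sets.
# The assembled attributes pass through shared Apollo functions for height-modulated distance mask fade, then world position offset wobble.
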Material /Game/Environments/Asteria/Props/Foliage/Foliage_LeafPile_A/Materials/M_Foliage_LeafPile_A_Parent
graph {
 properties {
  bUseMaterialAttributes=true
  bUsedWithNanite=true
  blendMode=BLEND_Masked
  pixelDepthOffsetMode=PDOM_Legacy
  twoSided=true
  usedShadingModels=Default Lit
 }
 nodes [
  n1 = SetMaterialAttributes!(
   MaterialAttributes
   Base Color=NamedRerouteUsage()
   Opacity Mask=NamedRerouteUsage()
   Metallic=Constant()
   Specular=NamedRerouteUsage()
   Roughness=Constant() {r=0.7}
   Normal=NamedRerouteUsage()
  ) {attributeSetTypes [
    #1
    #2
    #3
    #4
    #5
    #6
   ]}
  n2 = NamedRerouteUsage()
  MaterialFunctionCall!(In=MaterialFunctionCall!(In=@n1, HeightModulation=Power(Base=@n2) {constExponent=1}) {materialFunction=/Game/Environments/Apollo/Foliage/MaterialFunctions/MF_DistanceMaskFade}.Result) {materialFunction=/Game/Environments/Apollo/Foliage/MaterialFunctions/MF_WPOWobble}
  n3 = MaterialFunctionCall(Float2=TextureCoordinate() {coordinateIndex=1}) {materialFunction=&BreakOutFloat2Components}
  Multiply(A=@n3.R)
  n4 = TextureSampleParameter2D() {parameterName=HeightMask, texture=/Game/Environments/Asteria/Props/Foliage/Foliage_LeafPile_A/Textures/T_Leafpile_Grayscale_A_A_D}
  Multiply(A=@n4.R)
  n5 = TextureSampleParameter2D() {parameterName=Diffuse_2, texture=/Game/Environments/Asteria/Props/Foliage/Foliage_LeafPile_A/Textures/T_LeafPile_A_B_D}
  n6 = TextureSampleParameter2D() {parameterName=Diffuse, texture=/Game/Environments/Asteria/Props/Foliage/Foliage_LeafPile_A/Textures/T_LeafPile_A_A_D}
  NamedRerouteDeclaration(None=If(A=@n3.R, "A > B"=@n5.RGB, "A < B"=@n6.RGB) {constB=1}) {name=BaseColor, nodeColor=<rgba 0 0.7656 1 1>}
  NamedRerouteDeclaration(None=If(A=@n3.R, "A > B"=@n5.A, "A < B"=@n6.A) {constB=1}) {name=OpacityMask, nodeColor=<rgba 0 0.0391 1 1>}
  n7 = TextureCoordinate() {coordinateIndex=1}
  n8 = TextureSampleParameter2D() {
   parameterName=Normal_2
   samplerType=SAMPLERTYPE_Normal
   texture=/Game/Environments/Asteria/Props/Foliage/Foliage_LeafPile_A/Textures/T_NITE_Foliage_LeafPile_A_N
  }
  n9 = TextureSampleParameter2D() {
   parameterName=Normal
   samplerType=SAMPLERTYPE_Normal
   texture=/Game/Environments/Asteria/Props/Foliage/Foliage_LeafPile_A/Textures/T_LeafPile_A_A_N
  }
  NamedRerouteDeclaration(None=If(
    A=MaterialFunctionCall(Float2=@n7) {materialFunction=&BreakOutFloat2Components}.R
    "A > B"=@n8.RGB
    "A < B"=@n9.RGB
   ) {constB=1}) {name=Normal, nodeColor=<rgba 0 0.6719 1 1>}
  n10 = TextureCoordinate() {coordinateIndex=1}
  n11 = TextureSampleParameter2D() {
   parameterName=Specular_2
   samplerType=SAMPLERTYPE_LinearColor
   texture=/Game/Environments/Asteria/Props/Foliage/Foliage_LeafPile_A/Textures/T_LeafPile_A_B_S
  }
  n12 = TextureSampleParameter2D() {
   parameterName=Specular
   samplerType=SAMPLERTYPE_LinearColor
   texture=/Game/Environments/Asteria/Props/Foliage/Foliage_LeafPile_A/Textures/T_LeafPile_A_A_S
  }
  NamedRerouteDeclaration(None=If(
    A=MaterialFunctionCall(Float2=@n10) {materialFunction=&BreakOutFloat2Components}.R
    "A > B"=@n11.B
    "A < B"=@n12.B
   ) {constB=1}) {name=Spec, nodeColor=<rgba 1 0.4688 0 1>}
  n13 = TextureCoordinate() {coordinateIndex=1}
  n14 = TextureSampleParameter2D() {
   parameterName=HeightMask_2
   samplerType=SAMPLERTYPE_Masks
   texture=/Game/Environments/Asteria/Props/Foliage/Foliage_LeafPile_A/Textures/T_Leafpile_A_B_MASK
  }
  NamedRerouteDeclaration(None=If(
    A=MaterialFunctionCall(Float2=@n13) {materialFunction=&BreakOutFloat2Components}.R
    "A > B"=Power(Base=@n14.G)
    "A < B"=@n4.R
   ) {constB=1}) {name=Height, nodeColor=<rgba 1 0 0.7031 1>}
 ]
}
