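# Two-sided, Nanite-enabled parent material for the high-poly Asteria leaf pile, computing a camera-proximity vertex offset.
# The offset pulls vertices toward a point built from vertex color red and PivotOffset, scaled by ScaleMatch in instance space.
# A sphere mask fades it with camera distance, using radii that follow object scale, actor distance and a per-leaf vertex-color hash.
# Normal, S and D textures are sampled, but no material output connections appear.
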Material /Game/Environments/Asteria/Props/Foliage/Foliage_LeafPile_A/Materials/M_LeafPile_HighPoly_A_Parent
graph {
 properties {bUsedWithNanite=true, pixelDepthOffsetMode=PDOM_Legacy, twoSided=true, usedShadingModels=Default Lit}
 nodes [
  Constant()
  Constant() {r=0.7}
  n1 = MaterialFunctionCall(Float3=VertexColor()) {materialFunction=&BreakOutFloat3Components}
  n2 = Multiply(
   A=Add(
    A=MaterialFunctionCall(X=@n1.R, Y=@n1.R, Z=@n1.R) {materialFunction=&MakeFloat3}.Result
    B=VectorParameter() {parameterName=PivotOffset, defaultValue=<rgba -0.5 -0.5 -0.25 1>}.RGB
   )
   B=Constant3Vector() {constant=<rgba 1 1 1 0>}
  )
  n3 = ScalarParameter() {defaultValue=550, parameterName=ScaleMatch}
  n4 = Subtract(
   A=TransformPosition(None=Multiply(A=@n2, B=@n3) {constB=550}) {transformSourceType=TRANSFORMPOSSOURCE_Instance, transformType=TRANSFORMPOSSOURCE_World}
   B=WorldPosition().XYZ
  )
  n5 = Divide(A=Distance(A=CameraPositionWS(), B=WorldPosition().XYZ)) {constB=1000}
  n6 = Constant()
  n7 = Multiply(
   A=ScalarParameter() {defaultValue=0.25, parameterName=Scaling_Far}
   B=MaterialFunctionCall() {materialFunction=&ObjectScale}.Scale X
  )
  n8 = ScalarParameter() {defaultValue=0.5, parameterName=Scaling_Close}
  n9 = MaterialFunctionCall() {materialFunction=&ObjectScale}
  n10 = Divide(A=Distance(A=CameraPositionWS(), B=ActorPositionWS())) {constB=1000}
  n11 = Constant()
  n12 = Multiply(
   A=ScalarParameter() {defaultValue=0.6, parameterName=Scaling_Radius}
   B=MaterialFunctionCall() {materialFunction=&ObjectScale}.Scale X
  )
  n13 = Sine(None=Multiply(A=@n1.R, B=Add(A=@n1.G, B=@n1.G)))
  n14 = ScalarParameter() {defaultValue=659.55, parameterName=RandomSeed}
  n15 = Frac(None=Subtract(A=Multiply(A=@n13, B=@n14)) {constB=0.694})
  n16 = ScalarParameter() {defaultValue=0.15, parameterName=UnhideLeafAmount}
  n17 = LinearInterpolate(
   A=@n7
   B=LinearInterpolate(A=@n7, B=Multiply(A=@n8, B=@n9.Scale X), Alpha=SphereMask(A=@n10, B=@n11, Radius=@n12) {attenuationRadius=0.3, hardnessPercent=0})
   Alpha=OneMinus(None=Step(Y=@n15, X=@n16) {constX=0.1, constY=0.85})
  )
  n18 = SmoothStep(Value=Power(Base=SphereMask(A=@n5, B=@n6, Radius=@n17) {attenuationRadius=0.1, hardnessPercent=50}) {constExponent=0.5}) {constMax=0.8}
  n19 = Step(Y=@n15) {constX=0, constY=0.85}
  Reroute(None=Multiply(A=Multiply(A=@n4, B=@n18), B=OneMinus(None=@n19)))
  TextureSample() {samplerType=SAMPLERTYPE_Normal, texture=/Game/Environments/Asteria/Props/Foliage/Foliage_LeafPile_A/Textures/T_NITE_Foliage_LeafPile_A_N}
  TextureSample() {
   samplerType=SAMPLERTYPE_LinearColor
   texture=/Game/Environments/Asteria/Props/Foliage/Foliage_LeafPile_A/Textures/T_NITE_Foliage_LeafPile_A_S
  }
  TextureSample() {
   samplerType=SAMPLERTYPE_LinearColor
   texture=/Game/Environments/Asteria/Props/Foliage/Foliage_LeafPile_A/Textures/T_NITE_Foliage_LeafPile_A_D
  }
  VertexInterpolator(VS=TwoSidedSign())
 ]
}
